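# Niagara trail effect for the CubicVice glider: engine flames and glows, GPU embers, wing ribbons, a tail ribbon and speedlines.
# Engine emitters and embers attach to glider sockets; sidewings use anim trails, and the tail ribbon slerps along an interpolated owner axis.
# System timeline, startup curves and User.ForwardDot drive intensity, while RightDot skews speedlines.
# Front-end modifiers, variant color selectors and time-of-day scalars adapt the look per context.
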NiagaraSystem /BRCosmetics/Gadgets/Assets/VinderTech_GliderChute/Glider_CubicVice/FX/NS_Glider_CubicVice_Trail
dependencies {
 usedDataInterfaces [
  &NiagaraDataInterfaceCurve
  &NiagaraDataInterfaceVector2DCurve
  &NiagaraDataInterfaceColorCurve
  &NiagaraDataInterfaceArrayFloat3
  &NiagaraDataInterfaceArrayInt32
  &NiagaraDataInterfaceArrayFloat
  &NiagaraDataInterfaceSkeletalMesh
 ]
 usedDynamicInputs [&Boolean_AndOperation]
 usedModules [
  &self:SystemSpawnScript.SystemScriptSource.SystemScriptGraph.NiagaraNodeAssignment_0.SetVariables_#1
  &SystemState
  &FN_Integrator
  /BRCosmetics/Characters/Player/Female/Medium/Bodies/F_MED_PiperShelf_Pearl/FX/TimeOfDayGlobal
  &FN_VariantManager
  &self:SystemSpawnScript.SystemScriptSource.SystemScriptGraph.NiagaraNodeAssignment_1.SetVariables_#2
 ]
}
emitters {
 Embers {
  simTarget=GPUComputeSim
  emitterUpdateScript [
   &EmitterState()
   &SpawnRate(
    Spawn Probability=0.66
    SpawnRate=&Multiply_Float(A=20, B=&Lerp_Float(A=1, B=0, Alpha=$System.Timeline.NormalizedTime))
    Use Spawn Probability=true
   ) {module=SpawnRate001}
   &SpawnRate(
    Spawn Probability=&FloatFromCurve(
     FloatCurve=DataInterface() {propertyValues {Curve {keys [<at 0 0 RCIM_Constant>, <at 0.1 1>, {arriveTangent=-2.5249, leaveTangent=-2.5249, time=0.35, value=0}]}}}
     CurveIndex=$System.Timeline.Time
     Scale Curve=&FN_FrontEndFloatModifier(Gameplay=$User.ForwardDot, Front End=1)
    )
    SpawnRate=150
    Use Spawn Probability=true
   )
  ]
  particleSpawnScript [
   &PartitionParticles()
   /SetVariables(Particles.PerParticleRandom:NiagaraFloat=&RandomRangeFloat(Minimum=0, Maximum=1)) {isSetParameters=true}
   &InitializeParticle(
    Color Mode=::NewEnumerator0
    Lifetime=&Multiply_Float(
     A=&Lerp_Float(A=0.555, B=0.333, Alpha=$Particles.PerParticleRandom)
     B=&Lerp_Float(A=0.5, B=1, Alpha=$System.Timeline.NormalizedTime)
    )
    Sprite Size=&Multiply_Vector2D(
     A=<0.5 1>
     B=&MakeVector2D(
      X=&FloatFromCurve(
       FloatCurve=DataInterface() {propertyValues {Curve {keys [<at 0 1>,<at 0.8 2 RCIM_Constant>,<at 0.95 3 RCIM_Constant>]}}}
       CurveIndex=$Particles.PerParticleRandom
      )
      Y=&FloatFromCurve(
       FloatCurve=DataInterface() {propertyValues {Curve {keys [<at 0 1>,<at 0.8 2 RCIM_Constant>,<at 0.95 3>]}}}
       CurveIndex=$Particles.PerParticleRandom
      )
     )
    )
    Sprite Size Mode=::NewEnumerator4
    UsePositionOffset=false
    Write Material Random=true
   )
   &SkeletalMeshLocation(
    Bone / Socket Index=&SelectIntFromArray(Direct Array Index=$Particles.PartitionParticles.Partition, Int Selection Array=DataInterface() {propertyValues {IntData [4,5]}})
    MaskSpawnGroup=false
    Mesh Sampling Type=::NewEnumerator5
    Sampled Position Offset=&SelectVectorFromArray(
     Direct Array Index=$Particles.PartitionParticles.Partition
     Vector Selection Array=DataInterface() {propertyValues {FloatData [<0 0 0>, <0 0 0>]}}
    )
    Sampled Position Offset Coordinate Space:ENiagaraSkelMeshTransforms=::NewEnumerator0
    Skeletal Mesh=$System.Skeletal Mesh
    Socket Sampling Mode=::NewEnumerator5
   )
   &AddVelocity(
    Cone Angle=33
    Cone Axis=&FN_FrontEndVector3Modifier(Gameplay=<0 1 0.5>, Front End=<0 1 0>)
    Module.UseConeFalloff=false
    Module.UseVelDistribution=false
    Override Seed=false
    Velocity Mode=::NewEnumerator2
    Velocity Speed=&Multiply_Float(
     A=&Lerp_Float(A=1111, B=666, Alpha=$Particles.PerParticleRandom)
     B=&FN_FrontEndFloatModifier(Gameplay=&Lerp_Float(A=0.5, B=1, Alpha=$System.Timeline.NormalizedTime), Front End=1)
    )
   )
   &ShapeLocation(MaskSpawnGroup=false, Override Seed=false, Sphere Radius=1, UseOwnerScale=true)
   &CameraOffset(Camera Offset Amount=-15)
  ]
  particleUpdateScript [
   &ParticleState()
   &Drag(
    Drag=&FloatFromCurve(
     FloatCurve=DataInterface() {propertyValues {Curve {keys [<at 0 0.5>,<at 0.5 0.5>,<at 1 1>]}}}
     Scale Curve=&RandomRangeFloat(Minimum=5, Maximum=10)
    )
    Use Linear Drag=true
    Use Rotational Drag=false
   )
   /Niagara/Modules/Update/Forces/CurlNoiseForce(
    Noise Frequency=0.1
    Noise Strength=&FloatFromCurve(
     FloatCurve=DataInterface() {propertyValues {Curve {keys [{tangentMode=RCTM_User, time=0, value=0}, {tangentMode=RCTM_User, time=1, value=1}]}}}
     Scale Curve=&Lerp_Float(A=5555, B=222, Alpha=$Particles.PerParticleRandom)
    )
    Pan Noise=true
    Pan Noise Field=<0 0 1>
   )
   &InheritSourceMovement(ApplyForce=false, ApplyPosition=true, ApplyPositionFalloff=false)
   &SolveForcesAndVelocity(Clamp Velocity=false, Limit Acceleration=false)
   &ScaleColor(
    Scale Mode=::NewEnumerator1
    Scale RGBA=&MakeVector4_FromLinearColor(LinearColor=&VariantColorSelector(
      Variant Amount=0
      Original Color=&ColorFromCurve(ColorCurve=DataInterface() {propertyValues {
         AlphaCurve {keys [<at 0 1 RCIM_Cubic>]}
         BlueCurve {keys [<at 0 0 RCIM_Cubic>,<at 1 0 RCIM_Cubic>]}
         GreenCurve {keys [<at 0 9.1667 RCIM_Cubic>,<at 1 0 RCIM_Cubic>]}
         RedCurve {keys [<at 0 50 RCIM_Cubic>,<at 1 1 RCIM_Cubic>]}
        }})
      Variant 1 Color=&ColorFromCurve(ColorCurve=DataInterface() {propertyValues {
         AlphaCurve {keys [<at 0 1 RCIM_Cubic>]}
         BlueCurve {keys [<at 0 3.125 RCIM_Cubic>,<at 1 0.0524 RCIM_Cubic>]}
         GreenCurve {keys [<at 0 31.9296 RCIM_Cubic>,<at 1 0.5789 RCIM_Cubic>]}
         RedCurve {keys [<at 0 50 RCIM_Cubic>,<at 1 1 RCIM_Cubic>]}
        }})
      Variant 2 Color=?
     ))
    ScaleRGBA=true
   ) {module=ScaleColor001}
   &ScaleSpriteSizeBySpeed(
    Max Scale Factor=<1.666 13>
    Min Scale Factor=<1.333 1.333>
    Sample Scale Factor By Curve=true
    Scale Factor Curve=DataInterface() {propertyValues {Curve {keys [<at 0 0>, {arriveTangent=1.4989, interpMode=RCIM_Cubic, tangentMode=RCTM_Break, time=1, value=1}]}}}
    Velocity Threshold=666
   )
   &ScaleSpriteSize(
    Uniform Curve Scale=&MakeFloatFromVector(Vector=$Engine.Owner.Scale, Channel=::NewEnumerator0)
    Uniform Curve Sprite Scale=DataInterface() {propertyValues {Curve {keys [
        {tangentMode=RCTM_User, time=0, value=1}
        {arriveTangent=-0.0405, interpMode=RCIM_Cubic, leaveTangent=-0.0405, tangentMode=RCTM_User, time=0.5, value=1}
        {arriveTangent=-8.0089, interpMode=RCIM_Cubic, leaveTangent=-8.0089, tangentMode=RCTM_User, time=1, value=0}
       ]}}}
   )
   &ScaleSpriteSize(
    Scale Sprite Size Mode=::NewEnumerator0
    Uniform Scale Factor=&Lerp_Float(A=0.5, B=1, Alpha=$System.Timeline.NormalizedTime)
   ) {module=ScaleSpriteSize001}
  ]
  rendererClasses [&NiagaraSpriteRendererProperties]
  renderers [{rendererClass=&NiagaraSpriteRendererProperties}]
 }
 EngineFlames {
  emitterUpdateScript [
   &EmitterState(Life Cycle Mode=::NewEnumerator1, Loop Behavior=::NewEnumerator1, Loop Delay=$User.EngineDelay, Loop Duration=1, UseLoopDelay=true)
   &SpawnBurst_Instantaneous(Spawn Count=2, Use Spawn Probability=true)
   /SetVariables(
    Emitter.Flame1A:LinearColor=&VariantColorSelector(Variant Amount=0, Original Color=<rgba 0.125 0.0546 0.0216 1>, Variant 1 Color=<rgba 0.0677 0.0485 0.0043 1>, Variant 2 Color=?)
    Emitter.Flame1B:LinearColor=&VariantColorSelector(Variant Amount=0, Original Color=<rgba 0.4167 0.2258 0.0574 1>, Variant 1 Color=<rgba 0.4635 0.3605 0.2491 1>, Variant 2 Color=?)
    Emitter.Flame2A:LinearColor=&VariantColorSelector(Variant Amount=0, Original Color=<rgba 0.8021 0.2914 0 1>, Variant 1 Color=<rgba 1 1 1 1>, Variant 2 Color=?)
    Emitter.Flame2B:LinearColor=&VariantColorSelector(Variant Amount=0, Original Color=<rgba 0.375 0.1024 0 1>, Variant 1 Color=<rgba 0.4479 0.3443 0.0586 1>, Variant 2 Color=?)
    Emitter.Flame3A:LinearColor=&VariantColorSelector(Variant Amount=0, Original Color=<rgba 1 0.5297 0.18 1>, Variant 1 Color=<rgba 0.349 0.3056 0.0674 1>, Variant 2 Color=?)
    Emitter.Flame3B:LinearColor=&VariantColorSelector(Variant Amount=0, Original Color=<rgba 0.4583 0.1874 0 1>, Variant 1 Color=<rgba 0.4583 0.3875 0.0072 1>, Variant 2 Color=?)
    Emitter.FlameCoreA:LinearColor=&VariantColorSelector(Variant Amount=0, Original Color=<rgba 1 0.8076 0 1>, Variant 1 Color=<rgba 1 0.7557 0 1>, Variant 2 Color=?)
    Emitter.FlameCoreB:LinearColor=&VariantColorSelector(Variant Amount=0, Original Color=<rgba 0.4219 0.2144 0.0125 1>, Variant 1 Color=<rgba 0.125 0.1061 0.0423 1>, Variant 2 Color=?)
    Emitter.FlameDark:LinearColor=&VariantColorSelector(Variant Amount=0, Original Color=<rgba 0.1927 0.0905 0.0416 1>, Variant 1 Color=<rgba 0.2865 0.1851 0.0679 1>, Variant 2 Color=?)
    Emitter.FlameLight:LinearColor=&VariantColorSelector(Variant Amount=0, Original Color=<rgba 0.3854 0.1798 0.0781 1>, Variant 1 Color=<rgba 0.4583 0.4324 0.2053 1>, Variant 2 Color=?)
   ) {isSetParameters=true}
  ]
  particleSpawnScript [
   &PartitionParticles(ID to Partition:NiagaraInt32=$Engine.ExecIndex)
   &FN_SetSpawnGroup(Number of Spawn Groups=2)
   &InitializeParticle(
    Lifetime=0.25
    Mesh Scale=<0.24 0.24 0.24>
    Mesh Scale Mode=::NewEnumerator4
    Sprite Size Mode=::NewEnumerator3
    Uniform Sprite Size=220
    UsePositionOffset=false
    Write Material Random=true
   )
   &InitialMeshOrientation(
    Orientation Axis=&SelectVectorFromArray(
     Direct Array Index=$Particles.PartitionParticles.Partition
     Vector Selection Array=DataInterface() {propertyValues {FloatData [<0.1 -1 0>, <-0.1 1 0>]}}
    )
    Orientation Vector=<0 0 1>
    Use Orientation Vector=true
    Use Rotation Vector=false
   ) {module=InitialMeshOrientation001}
   /SetVariables(
    Particles.Flame2Scale:Vector3f=&Multiply_Vector(A=$Particles.Scale, B=<1 1 0.9>)
    Particles.FlameCoreScale:Vector3f=&Multiply_Vector(A=$Particles.Scale, B=<0.65 0.6 0.6>)
   ) {isSetParameters=true}
  ]
  particleUpdateScript [
   &ParticleState(Kill Particles When Lifetime Has Elapsed=false)
   &SkeletalMeshLocation(
    Apply Sampled Orientation to Mesh Particles=true
    Bone / Socket Index=&SelectIntFromArray(Direct Array Index=$Particles.SpawnGroup, Int Selection Array=DataInterface() {propertyValues {IntData [4,5]}})
    Inherit Velocity=false
    MaskSpawnGroup=false
    Mesh Sampling Type=::NewEnumerator5
    Sampled Position Offset=&SelectVectorFromArray(
     Direct Array Index=$Particles.PartitionParticles.Partition
     Vector Selection Array=DataInterface() {propertyValues {FloatData [<2 -7 0>, <-2 7 0>]}}
    )
    Sampled Position Offset Coordinate Space:ENiagaraMeshTransforms=::NewEnumerator3
    Skeletal Mesh=$System.Skeletal Mesh
    Socket Sampling Mode=::NewEnumerator5
   )
   &SolveForcesAndVelocity(Clamp Velocity=false, Limit Acceleration=false)
   &ScaleColor(
    Scale Alpha=&FloatFromCurve(FloatCurve=DataInterface() {propertyValues {Curve {keys [<at 0 0 RCIM_Cubic>,<at 0.5 1 RCIM_Cubic>]}}})
    Scale RGB=&MakeVectorFromLinearColorRGB(Color=&LerpLinearColors(StartColor=<rgba 1 1 1 1>, EndColor=<rgba 1.5 1.5 1.5 1>, LerpFactor=$System.Timeline.Time))
    ScaleA=true
    ScaleRGB=true
   )
   /SetVariables(
    Particles.Flame2Color:LinearColor=&MakeLinearColorFromVectorAndFloat("Vector (RGB)"=<1 1 1>, "Float (Alpha)"=&Lerp_Float(A=1, B=5, Alpha=$System.Timeline.Time))
    Particles.FlameCoreColor:LinearColor=&MakeLinearColorFromVectorAndFloat(
     "Vector (RGB)"=&MakeVectorFromLinearColorRGB(Color=&VariantColorSelector(Variant Amount=0, Original Color=<rgba 10 3.6336 0 1>, Variant 1 Color=<rgba 10 4.9278 0 1>, Variant 2 Color=?))
     "Float (Alpha)"=&Multiply_Float(
      A=&Lerp_Float(A=0.1, B=0.3, Alpha=1)
      B=&MakeCustomFloatFromBool(Bool=$User.bIsFullyDeployed, True Float=1, False Float=0)
     )
    )
   ) {isSetParameters=true}
   &ScaleMeshSize(Scale Factor=&MakeVector(
     X=&FloatFromCurve(FloatCurve=DataInterface() {propertyValues {Curve {keys [{tangentMode=RCTM_User, time=0, value=0}, {tangentMode=RCTM_User, time=0.5, value=1}]}}})
     Y=&FloatFromCurve(FloatCurve=DataInterface() {propertyValues {Curve {keys [{tangentMode=RCTM_User, time=0, value=0}, {tangentMode=RCTM_User, time=0.5, value=1}]}}})
     Z=&FN_FrontEndFloatModifier(
      Gameplay=&FloatFromCurve(FloatCurve=DataInterface() {propertyValues {Curve {keys [<at 0 0>,<at 0.85 1>]}}})
      Front End=$System.StartupCurve
     )
    ))
   &ScaleMeshSize(Scale Factor=&MakeVector(
     X=1
     Y=1
     Z=&Lerp_Float(
      A=&Lerp_Float(A=&FN_FrontEndFloatModifier(Gameplay=0.8, Front End=1.5), B=1.5, Alpha=$System.Timeline.Time)
      B=&SelectFloatFromArray(Direct Array Index=$Particles.PartitionParticles.Partition, Float Selection Array=DataInterface() {propertyValues {FloatData [2.4,2.4]}})
      Alpha=$System.StartupCurve
     )
    )) {module=ScaleMeshSize001}
   /SetVariables(
    Particles.Flame2Scale:Vector3f=&Multiply_Vector(
     A=$Particles.Initial.Flame2Scale
     B=&MakeVector(
      X=&FloatFromCurve(FloatCurve=DataInterface() {propertyValues {Curve {keys [<at 0 0>,<at 1 1>]}}})
      Y=1
      Z=1
     )
    )
    Particles.FlameCoreScale:Vector3f=&Multiply_Vector(
     A=$Particles.Initial.FlameCoreScale
     B=&MakeVector(
      X=&FloatFromCurve(FloatCurve=DataInterface() {propertyValues {Curve {keys [<at 0 0>,<at 1 1>]}}})
      Y=1
      Z=1
     )
    )
   ) {isSetParameters=true}
   /SetVariables(
    Particles.Flame2Scale:Vector3f=&Multiply_Vector(
     A=$Particles.Flame2Scale
     B=&MakeVector(
      X=&FloatFromCurve(FloatCurve=DataInterface() {propertyValues {Curve {keys [{tangentMode=RCTM_User, time=0, value=0}, {tangentMode=RCTM_User, time=0.5, value=1}]}}})
      Y=&FloatFromCurve(FloatCurve=DataInterface() {propertyValues {Curve {keys [{tangentMode=RCTM_User, time=0, value=0}, {tangentMode=RCTM_User, time=0.5, value=1}]}}})
      Z=&Multiply_Float(
       A=&Lerp_Float(
        A=&SelectFloatFromArray(Direct Array Index=$Particles.PartitionParticles.Partition, Float Selection Array=DataInterface() {propertyValues {FloatData [1.2,1.2]}})
        B=&SelectFloatFromArray(Direct Array Index=$Particles.PartitionParticles.Partition, Float Selection Array=DataInterface() {propertyValues {FloatData [4,4]}})
        Alpha=$System.StartupCurve
       )
       B=&Lerp_Float(
        A=&FN_FrontEndFloatModifier(Gameplay=1, Front End=$System.StartupCurve)
        B=2.5
        Alpha=$System.Timeline.Time
       )
      )
     )
    )
    Particles.FlameCoreScale:Vector3f=&Multiply_Vector(
     A=$Particles.FlameCoreScale
     B=&Multiply_Vector(
      A=&MakeVector(
       X=&FloatFromCurve(FloatCurve=DataInterface() {propertyValues {Curve {keys [{tangentMode=RCTM_User, time=0, value=0}, {tangentMode=RCTM_User, time=0.5, value=1}]}}})
       Y=1
       Z=&Lerp_Float(
        A=&SelectFloatFromArray(Direct Array Index=$Particles.PartitionParticles.Partition, Float Selection Array=DataInterface() {propertyValues {FloatData [1,1]}})
        B=&SelectFloatFromArray(Direct Array Index=$Particles.PartitionParticles.Partition, Float Selection Array=DataInterface() {propertyValues {FloatData [1.7,1.7]}})
        Alpha=$System.StartupCurve
       )
      )
      B=&MakeVector(
       X=1
       Y=1
       Z=&Lerp_Float(
        A=&FN_FrontEndFloatModifier(Gameplay=1, Front End=$System.StartupCurve)
        B=3
        Alpha=$System.Timeline.Time
       )
      )
     )
    )
   ) {isSetParameters=true}
   &ScaleSpriteSize()
   &DynamicMaterialParameters(
    Index 0 Param 1=&Lerp_Float(A=0.3, B=0.6, Alpha=1)
    Module.Param0WriteEnabled=true
    Module.Param1WriteEnabled=true
    Module.Param2WriteEnabled=true
    Module.Param3WriteEnabled=true
   )
  ]
  rendererClasses [&NiagaraMeshRendererProperties]
  renderers [
   {rendererClass=&NiagaraMeshRendererProperties, rendererIndex=0}
   {rendererClass=&NiagaraMeshRendererProperties, rendererIndex=1}
   {rendererClass=&NiagaraMeshRendererProperties, rendererIndex=2}
   {rendererClass=&NiagaraMeshRendererProperties, rendererIndex=3}
  ]
 }
 ExhaustGlow {
  emitterUpdateScript [
   &EmitterState(
    Inactive Response=::NewEnumerator1
    Life Cycle Mode=::NewEnumerator1
    Loop Behavior=::NewEnumerator1
    Loop Duration Mode=::NewEnumerator1
    UseLoopDelay=false
   )
   &SpawnBurst_Instantaneous(Spawn Count=2, Spawn Time=0.15, Use Spawn Probability=false, UseLoopCountLimit=false)
  ]
  particleSpawnScript [
   &PartitionParticles()
   &InitializeParticle(
    Position Offset=&SelectVectorFromArray(
     Direct Array Index=&SpawnGroup()
     Vector Selection Array=DataInterface() {propertyValues {FloatData [<77 54 28.5>, <77 -54 28.5>, <84 67 16.5>, <84 -67 16.5>]}}
    )
    Sprite Size Mode=::NewEnumerator3
    Uniform Sprite Size=25
    UsePositionOffset=true
    Write Material Random=true
   )
   &InitialMeshOrientation(
    Mesh Axis To Orient=&SelectVectorFromArray(
     Direct Array Index=$Particles.PartitionParticles.Partition
     Vector Selection Array=DataInterface() {propertyValues {FloatData [<0 0 1>, <0 0 -1>]}}
    )
    Mesh Orientation Mode=::NewEnumerator2
    Orientation Coordinate Space=::Local
    Orientation Vector=<0 1 0>
    Use Rotation Vector=false
   )
  ]
  particleUpdateScript [
   /SetVariables(Particles.Jitter:NiagaraFloat=&Waveform(
     Global Input Position=&RandomRangeFloat(Minimum=0, Maximum=1)
     "Waveform [1]"=::NewEnumerator12
     "[1] Frequency"=10
     "[1] Amplitude Min/Max"=<0.75 1>
    )) {isSetParameters=true}
   &ParticleState(Kill Particles When Lifetime Has Elapsed=false)
   &SkeletalMeshLocation(
    Bone / Socket Index=&SelectIntFromArray(Direct Array Index=$Particles.PartitionParticles.Partition, Int Selection Array=DataInterface() {propertyValues {IntData [4,5]}})
    MaskSpawnGroup=false
    Mesh Sampling Type=::NewEnumerator5
    Sampled Position Offset=&Multiply_VectorByFloat(
     Vector=&Lerp_Vector(A=<0 6 0>, B=<0 12 0>, Alpha=$System.Timeline.Time)
     Float=&SelectFloatFromArray(Direct Array Index=$Particles.PartitionParticles.Partition, Float Selection Array=DataInterface() {propertyValues {FloatData [1,-1]}})
    )
    Skeletal Mesh=$System.Skeletal Mesh
    Socket Sampling Mode=::NewEnumerator5
   )
   &ScaleColor(
    Scale Alpha=&Multiply_Float(A=&Multiply_Float(A=$Particles.Jitter, B=0.5), B=$System.EmissiveFadeInCurve)
    Scale RGB=&Multiply_VectorByFloat(
     Vector=&MakeVectorFromLinearColorRGB(Color=&VariantColorSelector(Variant Amount=0, Original Color=<rgba 1 0.3082 0.015 1>, Variant 1 Color=<rgba 0.7 0.433 0.0073 1>, Variant 2 Color=?))
     Float=&Lerp_Float(A=0.5, B=1, Alpha=$System.Timeline.Time)
    )
    ScaleA=true
    ScaleRGB=true
   )
   &FN_Light_Attributes(
    Color=$Particles.Color
    Create Color Attribute=true
    Create Exponent Attribute=true
    Create Position Attribute=true
    Create Radius Attribute=true
    Exponent=3
    Position=&AddVectorToPosition(
     Position=$Particles.SkeletalMeshLocation.SampledPosition
     Vector=&SelectVectorFromArray(
      Direct Array Index=$Particles.PartitionParticles.Partition
      Vector Selection Array=DataInterface() {propertyValues {FloatData [<0 20 0>, <0 20 0>]}}
     )
    )
    Radius=&Lerp_Float(A=40, B=60, Alpha=$System.Timeline.Time)
    Scale Color=&Multiply_Float(A=$Particles.Jitter, B=&Lerp_Float(A=30, B=70, Alpha=$System.Timeline.Time))
   )
   &FN_Sprite_Attributes(
    Camera Offset:NiagaraFloat=&FN_FrontEndFloatModifier(Gameplay=15, Front End=5)
    Color=&Multiply_LinearColorByFloat(
     Linear Color=&VariantColorSelector(Variant Amount=0, Original Color=<rgba 1 0.2686 0.017 0.675>, Variant 1 Color=<rgba 0.6771 0.5405 0.0734 0.675>, Variant 2 Color=?)
     Float=&Multiply_Float(
      A=&Multiply_Float(A=$Particles.Jitter, B=&Lerp_Float(A=0.2, B=0.3, Alpha=$System.Timeline.Time))
      B=$System.EmissiveFadeInCurve
     )
    )
    Create Camera Offset Attribute=true
    Create Color Attribute=true
    Create Position Attribute=true
    Create SpriteSize Attribute=true
    Position:NiagaraPosition=&AddVectorToPosition(
     Position=$Particles.SkeletalMeshLocation.SampledPosition
     Vector=&TransformVector(
      Vector=&FN_FrontEndVector3Modifier(Gameplay=<0 10 5>, Front End=<0 15 4>)
      Destination Space=::Simulation
      Transform Vector as Position=true
     )
    )
    SpriteSize=&FN_FrontEndVector2Modifier(Front End:Vector2f=<40 40>, Gameplay:Vector2f=<100 100>)
    Write to Module Namespace=true
   ) {module=FN_Sprite_Attributes001}
   &FN_Sprite_Attributes(
    Camera Offset:NiagaraFloat=&FN_FrontEndFloatModifier(Gameplay=15, Front End=15)
    Color=&Multiply_LinearColorByFloat(
     Linear Color=&VariantColorSelector(Variant Amount=0, Original Color=<rgba 2 0.6086 0 0.4>, Variant 1 Color=<rgba 2 1.6466 0 0.4>, Variant 2 Color=?)
     Float=&Multiply_Float(A=&Lerp_Float(A=0.2, B=0.33, Alpha=$System.Timeline.Time), B=$System.EmissiveFadeInCurve)
    )
    Create Camera Offset Attribute=true
    Create Color Attribute=true
    Create Position Attribute=true
    Create SpriteSize Attribute=true
    Position:NiagaraPosition=$Particles.SkeletalMeshLocation.SampledPosition
    SpriteSize=&Multiply_Vector2D_ByFloat(Float=&FN_FrontEndFloatModifier(Gameplay=1, Front End=1), Vector2D=<45 45>)
    Write to Module Namespace=true
   )
   &DynamicMaterialParameters(
    Index 1 Param 1=35
    Index 1 Param 2=5
    Module.Index0Param0Write=true
    Module.Index0Param1Write=true
    Module.Index0Param2Write=true
    Module.Index0Param3Write=true
    Module.Index1Param0Write=true
    Module.Index1Param1Write=true
    Module.Index1Param2Write=true
    Module.Index1Param3Write=true
    Write Parameter Index 1=true
   )
  ]
  rendererClasses [&NiagaraSpriteRendererProperties, &NiagaraLightRendererProperties]
  renderers [
   {rendererClass=&NiagaraSpriteRendererProperties}
   {rendererClass=&NiagaraLightRendererProperties, rendererIndex=1}
   {rendererClass=&NiagaraSpriteRendererProperties, rendererIndex=2}
   {rendererClass=&NiagaraSpriteRendererProperties, rendererIndex=3}
  ]
 }
 Sidewings {
  emitterUpdateScript [
   &EmitterState()
   &FN_SpawnAnimTrail_v2(Number of Anim Trails=2)
  ]
  particleSpawnScript [
   &PartitionParticles(ID to Partition:NiagaraInt32=$Engine.ExecIndex)
   /Game/Effects/Niagara/Modules/AnimTrail/FN_InitializeAnimTrail/FN_InitializeAnimTrailBoneIndecies(First Bone or Socket Index=1)
   &FN_InitializeAnimTrailBoneIndecies(AnimTrail Selection=1, First Bone or Socket Index=3, Second Bone or Socket Index=2) {module=FN_InitializeAnimTrailBoneIndecies001}
   &FN_InitializeAnimTrail(Color=<rgba 1 1 1 0.2>, Lifetime=0.125, SkeletalMesh=$System.Skeletal Mesh) {module=FN_InitializeAnimTrail001}
   &AddVelocity(Velocity=&RotateVectorByQuaternion(
     Vector To Rotate=&MakeVector(X=&FN_FrontEndFloatModifier(Gameplay=650, Front End=650), Y=0, Z=0)
     Quaternion=<0 0 1 1>
    ))
  ]
  particleUpdateScript [
   &InterpolateOverTime(Initial Float Value:NiagaraFloat=$System.Timeline.Time, Rate Of Change=2, Target Float Value=1, Use Set Interpolation Rate=true)
   &KillParticles(Kill Particles=&SetBoolByFloatComparison(A=$System.Timeline.Time, B=$Particles.InterpolateOverTime.InterpolateOverTime_Float.Moving Average, Comparison Type=::NewEnumerator6))
   &FN_UpdateAnimTrail(SkeletalMesh=$System.Skeletal Mesh) {module=FN_UpdateAnimTrail001}
   &InheritSourceMovement(
    Applied Position Scale=&Multiply_VectorByFloat(
     Vector=<1 1 1>
     Float=&FloatFromCurve(FloatCurve=DataInterface() {propertyValues {Curve {keys [
          {interpMode=RCIM_Cubic, tangentMode=RCTM_User, time=0.66, value=1}
          {arriveTangent=-1.9042, interpMode=RCIM_Cubic, leaveTangent=-1.9042, tangentMode=RCTM_User, time=1, value=0.5}
         ]}}})
    )
    ApplyForce=false
    ApplyPosition=true
    ApplyPositionFalloff=false
   )
   &SolveForcesAndVelocity(Clamp Velocity=false, Limit Acceleration=false)
   &FN_RibbonMaterial(
    Color Scale U=0.2182
    Erosion Amount:NiagaraFloat=&FloatFromCurve(FloatCurve=DataInterface() {propertyValues {Curve {keys [{tangentMode=RCTM_User, time=0, value=1}, {tangentMode=RCTM_User, time=1, value=0}]}}})
    Erosion Pan Offset U:NiagaraFloat=$Particles.MaterialRandom
    Erosion Pan Speed U=0.3
    Erosion Power=2
    Erossion Scale U=0.1
   )
   &ScaleColor(
    Scale Alpha=&FloatFromCurve(
     FloatCurve=DataInterface() {propertyValues {Curve {keys [<at 0 1 RCIM_Cubic>,<at 1 0>]}}}
     Scale Curve=&FN_FrontEndFloatModifier(Gameplay=$System.Timeline.Time, Front End=0.6)
    )
    Scale RGB=&VectorFromFloat(Value=&FNTimeOfDayScalar(Night Scale=0.4, Morning Scale=0.76, Evening Scale=0.65))
    ScaleA=true
    ScaleRGB=true
   )
   /SetVariables(Particles.RibbonU0Override=$Particles.NormalizedAge) {isSetParameters=true}
   /SetVariables(Particles.BlurColor:LinearColor=&MakeLinearColorFromVectorAndFloat(
     "Vector (RGB)"=&MakeVectorFromLinearColorRGB(Color=&MultiplyLinearColor(
       Linear Color=&LerpLinearColors(
        StartColor=<rgba 0.6458 0.6458 0.6458 1>
        EndColor=&LerpLinearColors(
         StartColor=<rgba 0.6875 0 0 1>
         EndColor=<rgba 0.3073 0 0 1>
         LerpFactor=&FN_FrontEndFloatModifier(Gameplay=0, Front End=1)
        )
        LerpFactor=$User.Variant
       )
       Scale Factor=&MakeVector4_FromFloat(Float=&FNTimeOfDayScalar(Night Scale=0.4, Morning Scale=0.7, Evening Scale=0.65))
      ))
     "Float (Alpha)"=&FloatFromCurve(
      FloatCurve=DataInterface() {propertyValues {Curve {keys [
          {arriveTangent=-2.1318, interpMode=RCIM_Cubic, leaveTangent=-2.1318, tangentMode=RCTM_User, time=0, value=1}
          {interpMode=RCIM_Cubic, tangentMode=RCTM_User, time=0.75, value=0}
         ]}}}
      Scale Curve=$System.Timeline.Time
     )
    )) {isSetParameters=true}
  ]
  rendererClasses [&NiagaraRibbonRendererProperties]
  renderers [{rendererClass=&NiagaraRibbonRendererProperties}]
 }
 Speedlines001 {
  emitterUpdateScript [
   &EmitterState()
   &SpawnRate(
    Spawn Probability=&MakeCustomFloatFromBool(
     Bool=&SetBoolByFloatComparison(A=$User.ForwardDot, B=0.05)
     True Float=1
     False Float=&FN_FrontEndFloatModifier(Gameplay=0, Front End=1)
    )
    SpawnRate=&MakeCustomFloatFromBool(Bool=$User.bIsFrontEndPreview, True Float=50, False Float=90)
    Use Spawn Probability=true
   ) {module=SpawnRate001}
  ]
  particleSpawnScript [
   &InitializeParticle(
    Color=&MakeLinearColorFromVectorAndFloat(
     "Vector (RGB)"=<1 1 1>
     "Float (Alpha)"=&Multiply_Float(
      A=&UniformRangedFloat(Minimum=1, Maximum=1)
      B=&MakeCustomFloatFromBool(
       Bool=$User.bIsFrontEndPreview
       True Float=&FloatFromCurve(
        FloatCurve=DataInterface() {propertyValues {Curve {keys [{tangentMode=RCTM_User, time=0, value=0}, {tangentMode=RCTM_User, time=2, value=1}]}}}
        CurveIndex=$Emitter.Age
        Scale Curve=0.4
       )
       False Float=1
      )
     )
    )
    Lifetime Max=0.2
    Lifetime Min=0.15
    Lifetime Mode=::NewEnumerator1
    Position Offset=&FN_FrontEndVector3Modifier(Gameplay=<0 -50 0>, Front End=<0 50 0>)
    Sprite Size Max=<3 10>
    Sprite Size Min=<2 2>
    Sprite Size Mode=::NewEnumerator2
    UsePositionOffset=true
    Write Material Random=true
   )
   &AddVelocity(
    Module.Use Falloff Distance=false
    Override Seed=false
    Velocity Mode=::NewEnumerator1
    Velocity Speed=&FN_FrontEndFloatModifier(Gameplay=-2500, Front End=2500)
   ) {module=AddVelocity001}
   &TorusLocation(
    Align Torus With Axis=true
    Coordinate Space:ENiagaraCoordinateSpace=::Simulation
    Handle Radius=0
    Large Radius:NiagaraFloat=&FN_FrontEndFloatModifier(
     Gameplay=&RandomRangeFloat(Minimum=450, Maximum=500)
     Front End=&RandomRangeFloat(Minimum=350, Maximum=400)
    )
    MaskSpawnGroup=false
    Module.Override Randomness=false
    Offset=<0 33 0>
    Torus Axis=<0 1 0>
    Torus Origin:NiagaraPosition=&AddVectorToPosition(
     Position=$Particles.Position
     Vector=&FN_FrontEndVector3Modifier(Gameplay=<0 -500 0>, Front End=<0 -500 0>)
    )
   )
   &AddVelocity(
    Velocity=&Add_Vector(
     A=&RandomRangeVector(Minimum=<0 300 0>, Maximum=<0 700 1>)
     B=&Add_Vector(
      A=&Multiply_VectorByFloat(Vector=<-555 0 0>, Float=$User.RightDot)
      B=&Multiply_VectorByFloat(Vector=<0 0 222>, Float=$User.ForwardDot)
     )
    )
    Velocity Speed Scale=&FN_FrontEndFloatModifier(Gameplay=1, Front End=-1)
   )
  ]
  particleUpdateScript [
   &SolveForcesAndVelocity(Clamp Velocity=false, Limit Acceleration=false)
   &ParticleState()
   &ScaleSpriteSize(Scale Factor=&Vector2DFromCurve(
     Vector2Curve=DataInterface() {propertyValues {XCurve {keys [<at 0 0.5 RCIM_Cubic>,<at 0.5 1 RCIM_Cubic>]}, YCurve {keys [<at 0 0 RCIM_Cubic>,<at 0.5 1 RCIM_Cubic>]}}}
     ScaleCurve=&Vector2DFromFloat(Value=1)
    ))
   &ScaleSpriteSizeBySpeed(Max Scale Factor=<1 50>, Min Scale Factor=<0.01 0.05>, Velocity Threshold=1)
   &ScaleColor(
    Scale Alpha=&FloatFromCurve(
     FloatCurve=DataInterface() {propertyValues {Curve {keys [
         <at 0 0 RCIM_Cubic>
         <at 0.3759 1.0207>
         {arriveTangent=-2.221, interpMode=RCIM_Cubic, leaveTangent=-2.221, tangentMode=RCTM_User, time=1, value=0}
        ]}}}
     Scale Curve=&MakeCustomFloatFromBool(
      Bool=$User.bIsFrontEndPreview
      True Float=&MakeCustomFloatFromBool(Bool=$User.bIsFullyDeployed, True Float=0.5, False Float=0)
      False Float=&Lerp_Float(A=0, B=0.5, Alpha=&FN_FrontEndFloatModifier(Gameplay=$System.Timeline.Time, Front End=0.5))
     )
    )
    Scale RGB=&VectorFromFloat(Value=&FN_FrontEndFloatModifier(Gameplay=0.5, Front End=0.33))
    ScaleA=true
    ScaleRGB=true
   )
  ]
  rendererClasses [&NiagaraSpriteRendererProperties]
  renderers [{rendererClass=&NiagaraSpriteRendererProperties}]
 }
 TailTrail {
  emitterUpdateScript [
   &EmitterState(
    Inactive Response=::NewEnumerator1
    Life Cycle Mode=::NewEnumerator1
    Loop Behavior=::NewEnumerator1
    Loop Duration Mode=::NewEnumerator1
    UseLoopDelay=false
   )
   &Timeline(Play=&SetBoolByFloatComparison(A=$User.ForwardDot, B=0.1))
   &InterpolateOverTime(Target Float Value:NiagaraFloat=$User.RightDot)
   &SampleSkeletalMesh(
    AdditionalYawPitchRoll=false
    Bone / Socket Index=6
    Mesh Sampling Type:ENiagaraSkelSamplingModeFull=::NewEnumerator5
    Skeletal Mesh:NiagaraDataInterfaceSkeletalMesh=$System.Skeletal Mesh
    Socket Sampling Mode:ENiagaraSocketSamplingMode=::NewEnumerator5
   )
   /SetVariables(
    Emitter.PositionOffset:Vector3f=&TransformVector(Vector=$StackContext.SampleSkeletalMesh.SampledPosition, Destination Space=::Simulation)
    Emitter.RibbonColor:LinearColor=&LerpLinearColors(
     StartColor=<rgba 1 0 0.0067 0.15>
     EndColor=<rgba 1 0 0 0.088>
     LerpFactor=&FN_FrontEndFloatModifier(Gameplay=0, Front End=1)
    )
    Emitter.RibbonLength:NiagaraFloat=&FN_FrontEndFloatModifier(Gameplay=&Lerp_Float(A=30, B=100, Alpha=$StackContext.Timeline.Time), Front End=70)
   ) {isSetParameters=true}
   &SpawnBurst_Instantaneous(Spawn Count=4, Use Spawn Probability=false, UseLoopCountLimit=false)
   /BRCosmetics/Animation/Game/MainPlayer/Emotes/UpbeatIguana/FX/InterpolateTrailVector(Range=0.3, Target Vector:Vector3f=$Engine.Owner.SystemYAxis)
  ]
  particleSpawnScript [/SetVariables(Particles.RibbonWidth=&Multiply_Float(A=5, B=&Lerp_Float(A=0.66, B=1, Alpha=&ReturnNormalizedExecIndex()))) {isSetParameters=true}]
  particleUpdateScript [
   /SetVariables(Particles.Color=&MultiplyLinearColor(
     Linear Color=$Emitter.RibbonColor
     Scale Factor=&MakeVector4_FromFloat(Float=&Multiply_Float(
       A=$System.Timeline.Time
       B=&FNTimeOfDayScalar(Night Scale=0.5, Morning Scale=0.8, Evening Scale=0.8)
      ))
    )) {isSetParameters=true}
   /SetVariables(
    Particles.InterpolatedTrailVector:Vector3f=&TransformVector(
     Vector=/Niagara/DynamicInputs/Math/SlerpVector(
      Vector A:Vector3f=$Emitter.InterpolatedVector
      Vector B:Vector3f=$Emitter.TargetVector
      Alpha:NiagaraFloat=&ReturnNormalizedExecIndex()
      Scale:NiagaraFloat=1
     )
     Source Space=::World
     Destination Space=::Simulation
     Apply Scale=true
    )
    Particles.RibbonFacing=&TransformVector(Vector=$Engine.Owner.SystemZAxis, Source Space=::World, Destination Space=::Simulation)
   ) {isSetParameters=true}
   /SetVariables(Particles.SidewaysOffsetVector:Vector3f=&Multiply_VectorByFloat(
     Vector=&NormalizeVector(
      Vector To Normalize=/Niagara/DynamicInputs/Vectors/CrossProduct(
       Vector A:Vector3f=$StackContext.InterpolatedTrailVector
       Vector B:Vector3f=$Particles.RibbonFacing
       Normalize Incoming Vectors:NiagaraBool=true
      )
      Safe Normalize=true
      Safe Fallback Threshold=0
      Safe Fallback Vector=<0 0 0.001>
     )
     Float=&Multiply_Float(
      A=$Emitter.InterpolateOverTime.InterpolateOverTime_Float.Moving Average
      B=&Multiply_Float(
       A=&Power(a=&OneMinusFloat(Float=&ReturnNormalizedExecIndex()), b=2)
       B=0
      )
     )
    )) {isSetParameters=true}
   /SetVariables(Particles.InterpolatedPosition:NiagaraPosition=&AddVectorToPosition(
     Position=&AddVectorToPosition(
      Position=$Engine.Emitter.SimulationPosition
      Vector=&Add_Vector(A=$Emitter.PositionOffset, B=$StackContext.SidewaysOffsetVector)
     )
     Vector=&Multiply_VectorByFloat(
      Vector=$Particles.InterpolatedTrailVector
      Float=&Multiply_Float(A=&OneMinusFloat(Float=&ReturnNormalizedExecIndex()), B=$Emitter.RibbonLength)
     )
    )) {isSetParameters=true}
  ]
  rendererClasses [&NiagaraRibbonRendererProperties]
  renderers [{rendererClass=&NiagaraRibbonRendererProperties}]
 }
}
summary {
 systemName=NS_Glider_CubicVice_Trail
 emitters [
  {emitterName=Speedlines001, rendererClasses [&NiagaraSpriteRendererProperties]}
  {emitterName=ExhaustGlow, rendererClasses [&NiagaraSpriteRendererProperties, &NiagaraLightRendererProperties]}
  {emitterName=Sidewings, rendererClasses [&NiagaraRibbonRendererProperties]}
  {emitterName=Embers, simTarget=GPUComputeSim, rendererClasses [&NiagaraSpriteRendererProperties]}
  {emitterName=EngineFlames, rendererClasses [&NiagaraMeshRendererProperties]}
  {emitterName=TailTrail, rendererClasses [&NiagaraRibbonRendererProperties]}
 ]
 userVariables [
  User.bIsFrontEnd {type=NiagaraBool}
  User.bIsFrontEndPreview {type=NiagaraBool}
  User.bIsFullyDeployed {type=NiagaraBool}
  User.EngineDelay {type=NiagaraFloat, defaultValue=0.16, description=""}
  User.ForwardDot {type=NiagaraFloat, defaultValue=0, description=""}
  User.RibbonIDOffset {type=NiagaraInt32}
  User.RightDot {type=NiagaraFloat, defaultValue=0, description=""}
  User.Variant {type=NiagaraFloat, defaultValue=0, description=""}
 ]
}
