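# Niagara intro effect for the CubicVice glider, drawing short-lived light and lightning trails plus flashes from the glider's skeletal mesh.
# LightTrail and LightningTrail are two-bone AnimTrail ribbons that inherit source movement and fade over a system Timeline, with LightTrail also gated by User.TrailCurve.
# LightningFlashes spawns SubUV sprites around two bones, with probability from User.SpawnCurve and smaller front-end sizes.
# The variant switches LightTrail colors, while a waveform flickers the lightning ribbons.
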NiagaraSystem /BRCosmetics/Gadgets/Assets/VinderTech_GliderChute/Glider_CubicVice/FX/NS_Glider_Cubicvice_Intro
dependencies {
 usedDataInterfaces [
  &NiagaraDataInterfaceSpriteRendererInfo
  &NiagaraDataInterfaceArrayInt32
  &NiagaraDataInterfaceCurve
  &NiagaraDataInterfaceColorCurve
  &NiagaraDataInterfaceSkeletalMesh
  &NiagaraDataInterfaceArrayFloat
 ]
 usedDynamicInputs []
 usedModules [
  &self:SystemSpawnScript.SystemScriptSource.SystemScriptGraph.NiagaraNodeAssignment_0.SetVariables_#1
  &SystemState
  &Timeline
  &FN_VariantManager
 ]
}
emitters {
 LightTrail {
  emitterSpawnScript [&FN_RibbonIDOffsetUtility()]
  emitterUpdateScript [
   &EmitterState(Life Cycle Mode=::NewEnumerator1, Loop Behavior=::NewEnumerator1, Loop Delay=0.125, Loop Duration=3, UseLoopDelay=true)
   &FN_SpawnAnimTrail(
    Enable:NiagaraBool=&Boolean_AndOperation(
     Bool1=&SetBoolByFloatComparison(A=$System.Timeline.Float, B=0.03)
     Bool2=&SetBoolByFloatComparison(A=$User.TrailCurve, B=1, Comparison Type=::NewEnumerator4)
    )
    Number of Anim Trails=2
   )
  ]
  particleSpawnScript [
   &FN_InitializeAnimTrailBoneIndecies(First Bone or Socket Index=4, Second Bone or Socket Index=4)
   &FN_InitializeAnimTrailBoneIndecies(AnimTrail Selection=1, First Bone or Socket Index=5, Second Bone or Socket Index=5) {module=FN_InitializeAnimTrailBoneIndecies001}
   &FN_InitializeAnimTrail(AnimTrailWidth=12, Lifetime=0.2, SkeletalMesh=$System.Skeletal Mesh)
   &AddVelocity(Velocity=<0 400 100>)
  ]
  particleUpdateScript [
   &InheritSourceMovement(
    Applied Force Scale=<0.5 0.5 1>
    Applied Position Scale=&Multiply_VectorByFloat(
     Vector=<1 1 1>
     Float=&FloatFromCurve(FloatCurve=DataInterface() {propertyValues {Curve {keys [{interpMode=RCIM_Cubic, tangentMode=RCTM_User, time=0, value=1}, {interpMode=RCIM_Cubic, tangentMode=RCTM_User, time=1, value=0.425}]}}})
    )
    ApplyForce=false
    ApplyPosition=true
    ApplyPositionFalloff=false
   )
   &SolveForcesAndVelocity(Clamp Velocity=false, Limit Acceleration=false)
   &ScaleColor(
    Scale Alpha=&FloatFromCurve(
     FloatCurve=DataInterface() {propertyValues {Curve {keys [{interpMode=RCIM_Cubic, tangentMode=RCTM_User, time=0, value=1}, {interpMode=RCIM_Cubic, tangentMode=RCTM_User, time=0.9, value=0}]}}}
     Scale Curve=$System.Timeline.Float
    )
    Scale RGB=&Lerp_Vector(
     A=&MakeVectorFromLinearColorRGB(Color=&MakeCustomLinearColorFromBool(Bool=$System.bIsVariant0, True Color=<rgba 1 0.5073 0 1>, False Color=<rgba 0.4948 0.4541 0.268 1>))
     B=&MakeVectorFromLinearColorRGB(Color=&MakeCustomLinearColorFromBool(Bool=$System.bIsVariant0, True Color=<rgba 0.5052 0.1406 0 1>, False Color=<rgba 0.2917 0.2603 0 1>))
     Alpha=&FloatFromCurve(FloatCurve=DataInterface() {propertyValues {Curve {keys [{tangentMode=RCTM_User, time=0, value=0}, {tangentMode=RCTM_User, time=1, value=1}]}}})
    )
    ScaleA=true
    ScaleRGB=true
   )
   &FN_UpdateAnimTrail(Locked Particle RibbonWidth=1, SkeletalMesh=$System.Skeletal Mesh) {module=FN_UpdateAnimTrail001}
   &ScaleColor(ScaleA=true, ScaleRGB=true) {module=ScaleColor001}
  ]
  rendererClasses [&NiagaraRibbonRendererProperties]
  renderers [{rendererClass=&NiagaraRibbonRendererProperties}]
 }
 LightningFlashes {
  emitterUpdateScript [
   &EmitterState(Life Cycle Mode=::NewEnumerator1, Loop Behavior=::NewEnumerator1, Loop Delay=0.2, Loop Duration=1.8, UseLoopDelay=true)
   &SpawnRate(Spawn Probability=$User.SpawnCurve, SpawnRate=15, Use Spawn Probability=true)
  ]
  particleSpawnScript [
   &PartitionParticles()
   &InitializeParticle(
    Color Mode=::NewEnumerator0
    Lifetime=0.05
    Sprite Rotation Mode=::NewEnumerator2
    Sprite Size Mode=::NewEnumerator3
    Uniform Sprite Size=&FN_FrontEndFloatModifier(Gameplay=30, Front End=20)
    UsePositionOffset=false
    Write Material Random=true
   )
   &SubUVAnimation(
    "Sprite Renderer "=DataInterface() {propertyValues {SpriteRenderer=&self:LightningFlashes_1.NiagaraSpriteRendererProperties_0}}
    SubUV Animation Mode:ENiagara_SubUVLookupModeV2=::NewEnumerator2
    UseEndFrame=false
    UseStartFrame=false
   )
   &SkeletalMeshLocation(
    Bone / Socket Index=&SelectIntFromArray(Direct Array Index=$Particles.PartitionParticles.Partition, Int Selection Array=DataInterface() {propertyValues {IntData [0,1]}})
    Bone Sampling Mode=::NewEnumerator4
    MaskSpawnGroup=false
    Skeletal Mesh=$System.Skeletal Mesh
   )
   &ShapeLocation(
    Cylinder Height=20
    Cylinder Radius=30
    MaskSpawnGroup=false
    Non Uniform Scale=$Engine.Owner.Scale
    Override Seed=false
    Rotation Mode=::NewEnumerator1
    Shape Primitive=::NewEnumerator1
    Surface Only=true
    Use Endcaps In Surface Only Mode=false
    UseOwnerScale=true
   )
   &CameraOffset(Camera Offset Amount=15)
   /SetVariables(
    Particles.GlowOffset=30
    Particles.GlowSize=&Multiply_Vector2D_ByFloat(Float=&FN_FrontEndFloatModifier(Gameplay=1.5, Front End=1.3), Vector2D=$Particles.SpriteSize)
   ) {isSetParameters=true}
   /SetVariables(Particles.DecalDynamic:Vector4f=<0.2 0 0 0>, Particles.DecalDynamic1:Vector4f=<75 15 0 0>) {isSetParameters=true}
  ]
  particleUpdateScript [
   &ParticleState()
   &ScaleColor(
    Scale RGB=&MakeVectorFromLinearColorRGB(Color=&ColorFromCurve(ColorCurve=DataInterface() {propertyValues {
        AlphaCurve {keys [<at 0 1 RCIM_Cubic>]}
        BlueCurve {keys [<at 0 0 RCIM_Cubic>,<at 0.5889 0.214 RCIM_Cubic>]}
        GreenCurve {keys [<at 0 5 RCIM_Cubic>,<at 0.5889 0.6692 RCIM_Cubic>]}
        RedCurve {keys [<at 0 4.9897 RCIM_Cubic>,<at 0.5889 1 RCIM_Cubic>]}
       }}))
    ScaleA=true
    ScaleRGB=true
   )
   /SetVariables(Particles.GlowColor=&Multiply_LinearColorByFloat(Linear Color=$Particles.Color, Float=0.1)) {isSetParameters=true}
  ]
  rendererClasses [&NiagaraSpriteRendererProperties]
  renderers [
   {rendererClass=&NiagaraSpriteRendererProperties}
   {rendererClass=&NiagaraSpriteRendererProperties, rendererIndex=1}
  ]
 }
 LightningTrail {
  emitterSpawnScript [&FN_RibbonIDOffsetUtility()]
  emitterUpdateScript [
   &EmitterState(Life Cycle Mode=::NewEnumerator1, Loop Behavior=::NewEnumerator1, Loop Delay=0.125, Loop Duration=2.5, UseLoopDelay=true)
   &FN_SpawnAnimTrail(
    Enable:NiagaraBool=&SetBoolByFloatComparison(A=$System.Timeline.Float, B=0.05)
    Number of Anim Trails=2
    Spawn Mode:ENiagaraSkeletalSpawning=::NewEnumerator0
   )
  ]
  particleSpawnScript [
   &FN_InitializeAnimTrailBoneIndecies(Second Bone or Socket Index=0)
   &FN_InitializeAnimTrailBoneIndecies(AnimTrail Selection=1, First Bone or Socket Index=1) {module=FN_InitializeAnimTrailBoneIndecies001}
   &FN_InitializeAnimTrail(AnimTrailWidth=40, SkeletalMesh=$System.Skeletal Mesh)
   &AddVelocity(Velocity=<0 200 -50>)
   &DynamicMaterialParameters(
    Index 0 Param 2=&FloatFromCurve(FloatCurve=DataInterface() {propertyValues {Curve {keys [{tangentMode=RCTM_User, time=0, value=0}, {tangentMode=RCTM_User, time=1, value=1}]}}})
    Module.Index0Param0Write=true
    Module.Index0Param1Write=true
    Module.Index0Param2Write=true
    Module.Index0Param3Write=true
   )
  ]
  particleUpdateScript [
   &InheritSourceMovement(
    Applied Force Scale=<0.5 0.5 1>
    Applied Position Scale=&Multiply_VectorByFloat(
     Vector=<1 1 1>
     Float=&FloatFromCurve(FloatCurve=DataInterface() {propertyValues {Curve {keys [
          {arriveTangent=-1.3574, interpMode=RCIM_Cubic, leaveTangent=-1.3574, tangentMode=RCTM_User, time=0, value=1}
          {interpMode=RCIM_Cubic, tangentMode=RCTM_User, time=0.5, value=0.25}
         ]}}})
    )
    ApplyForce=false
    ApplyPosition=true
    ApplyPositionFalloff=false
   )
   &SolveForcesAndVelocity(Clamp Velocity=false, Limit Acceleration=false)
   &ScaleColor(
    Scale Alpha=&FloatFromCurve(
     FloatCurve=DataInterface() {propertyValues {Curve {keys [{interpMode=RCIM_Cubic, tangentMode=RCTM_User, time=0, value=1}, {interpMode=RCIM_Cubic, tangentMode=RCTM_User, time=1, value=0}]}}}
     Scale Curve=$System.Timeline.Float
    )
    Scale RGB=&Multiply_VectorByFloat(
     Vector=&MakeVectorFromLinearColorRGB(Color=&ColorFromCurve(ColorCurve=DataInterface() {propertyValues {
         AlphaCurve {keys [<at 0 1 RCIM_Cubic>]}
         BlueCurve {keys [<at 0 0 RCIM_Cubic>,<at 0.6344 0 RCIM_Cubic>]}
         GreenCurve {keys [<at 0 3.1341 RCIM_Cubic>,<at 0.6344 0.3584 RCIM_Cubic>]}
         RedCurve {keys [<at 0 5 RCIM_Cubic>,<at 0.6344 1 RCIM_Cubic>]}
        }}))
     Float=&Waveform("Waveform [1]"=::NewEnumerator12, "[1] Frequency"=25, "[1] Amplitude Min/Max"=<0 1>)
    )
    ScaleA=true
    ScaleRGB=true
   )
   &FN_UpdateAnimTrail(Locked Particle RibbonWidth=1, SkeletalMesh=$System.Skeletal Mesh) {module=FN_UpdateAnimTrail001}
   /SetVariables(Particles.LightningGlowColor:LinearColor=&Multiply_LinearColorByFloat(Linear Color=$Particles.Color, Float=0.15)) {isSetParameters=true}
  ]
  rendererClasses [&NiagaraRibbonRendererProperties]
  renderers [
   {rendererClass=&NiagaraRibbonRendererProperties}
   {rendererClass=&NiagaraRibbonRendererProperties, rendererIndex=1}
  ]
 }
}
summary {
 systemName=NS_Glider_Cubicvice_Intro
 emitters [
  {emitterName=LightningFlashes, rendererClasses [&NiagaraSpriteRendererProperties]}
  {emitterName=LightTrail, rendererClasses [&NiagaraRibbonRendererProperties]}
  {emitterName=LightningTrail, rendererClasses [&NiagaraRibbonRendererProperties]}
 ]
 userVariables [
  User.bIsFrontEnd {type=NiagaraBool}
  User.bIsFrontEndPreview {type=NiagaraBool}
  User.ForwardDot {type=NiagaraFloat, defaultValue=0, description=""}
  User.RibbonIDOffset {type=NiagaraInt32}
  User.RightDot {type=NiagaraFloat, defaultValue=0, description=""}
  User.SpawnCurve {type=NiagaraFloat, defaultValue=0, description=""}
  User.TrailCurve {type=NiagaraFloat, defaultValue=0, description=""}
  User.Variant {type=NiagaraFloat, defaultValue=1, description=""}
 ]
}
